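# AthenaMainStatTile is a FortProfileStatsMainTile widget that displays a main profile stat in the Athena frontend.
# A fixed-size SizeBox holds an Overlay with a translucent glow, a large centered value and a smaller name along the bottom.
# Both text blocks are inherited from the parent class.
# SetStatValue uppercases the incoming text before writing it to ValueTextBlock; the event graph is empty.
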WidgetBlueprint /Game/Athena/UI/Frontend/AthenaMainStatTile
graphs [
 EventGraph {links [], nodes []}
 SetStatValue {nodes [
   n1 = SetStatValue!() -> (then, StatValueAsText:Text) {is=title K2Node_FunctionEntry}
   > "SetText (Text)"(
    self=Get ValueTextBlock() -> (ValueTextBlock:Common Text Object Reference)
    InText=TextToUpper!(InText:"Text (by ref)"=@n1) -> (ReturnValue:Text)
   )
  ]}
]
variables []
widgets {
 info {inheritedWidgetCount=2, parentClass=/Script/FortniteUI.FortProfileStatsMainTile, rootWidgetClass=&SizeBox, widgetCount=5}
 widgets [
  SizeBox_16 {
   widgetClassPath=&SizeBox
   widget=SizeBox_16
   properties {
    bOverride_MinDesiredHeight=true
    bOverride_MinDesiredWidth=true
    heightOverride=192
    minDesiredHeight=192
    minDesiredWidth=352
    widthOverride=352
   }
  }
  Overlay_0 {widgetClassPath=&Overlay, parent=SizeBox_16, widget=Overlay_0, slot {class=SizeBoxSlot}}
  Glow {
   widgetClassPath=&Image
   bIsVariable=true
   parent=Overlay_0
   widget=Glow
   properties {
    visibility=SelfHitTestInvisible
    brush {
     bIsDynamicallyLoaded=false
     drawAs=Image
     imageType=NoImage
     mirroring=NoMirror
     resourceName=None
     resourceObject=/FortUILibrary/Styles/Border/Materials/M_Border_SoftWhiteGlow
     tiling=NoTile
     imageSize=<256 156>
     margin=<ltrb 0 0 0 0>
     outlineSettings {
      bUseBrushTransparency=false
      roundingType=HalfHeightRadius
      width=0
      color {colorUseRule=UseColor_Specified, specifiedColor=<rgba 0 0 0 0>}
      cornerRadii=<0 0 0 0>
     }
     tintColor {colorUseRule=UseColor_Specified, specifiedColor=<rgba 1 1 1 1>}
     uVRegion {bIsValid=false, max=<0 0>, min=<0 0>}
    }
    colorAndOpacity=<rgba 1 1 1 0.4>
   }
   slot {class=OverlaySlot, properties {horizontalAlignment=HAlign_Center}}
  }
  ValueTextBlock {
   widgetClassPath=&CommonTextBlock
   bInherited=true
   bIsVariable=true
   parent=Overlay_0
   widget=ValueTextBlock
   properties {
    justification=Center
    style=/FortUILibrary/Styles/Typography/Foundation/TextStyle-Header-XL-Horizon.TextStyle-Header-XL-Horizon_C
    text=N/A
    font {
     bForceMonospaced=false
     bMaterialIsStencil=false
     fontMaterial=/FortUILibrary/Materials/Foundation/M_Font_Horizon_Dynamic
     fontObject=&Burbank
     letterSpacing=0
     monospacedWidth=1
     size=72
     skewAmount=0
     typefaceFontName=Burbank-Black
     outlineSettings {
      bApplyOutlineToDropShadows=true
      bMiteredCorners=false
      bSeparateFillAlpha=false
      outlineMaterial=null
      outlineSize=3
      outlineColor=<rgba 0 0 0 0.1>
     }
    }
    shadowOffset=<0 0>
   }
   slot {class=OverlaySlot, properties {horizontalAlignment=HAlign_Center, verticalAlignment=VAlign_Center, padding=<ltrb 0 0 0 16>}}
  }
  NameTextBlock {
   widgetClassPath=&CommonTextBlock
   bInherited=true
   bIsVariable=true
   parent=Overlay_0
   widget=NameTextBlock
   properties {
    justification=Center
    scrollStyle=/FortUILibrary/Styles/Typography/Foundation/TextScrollStyle-NoFade.TextScrollStyle-NoFade_C
    style=/FortUILibrary/Styles/Typography/Foundation/TextStyle-Header-S-70pc.TextStyle-Header-S-70pc_C
    text="{STAT NAME}"
    colorAndOpacity {colorUseRule=UseColor_Specified, specifiedColor=<rgba 1 1 1 0.7>}
    font {
     bForceMonospaced=false
     bMaterialIsStencil=false
     fontMaterial=null
     fontObject=&Burbank
     letterSpacing=0
     monospacedWidth=1
     size=24
     skewAmount=0
     typefaceFontName=Burbank-Black
     outlineSettings {
      bApplyOutlineToDropShadows=false
      bMiteredCorners=false
      bSeparateFillAlpha=false
      outlineMaterial=null
      outlineSize=0
      outlineColor=<rgba 0 0 0 1>
     }
    }
    shadowOffset=<0 0>
   }
   slot {class=OverlaySlot, properties {horizontalAlignment=HAlign_Center, verticalAlignment=VAlign_Bottom, padding=<ltrb 10 0 10 24>}}
  }
 ]
}
