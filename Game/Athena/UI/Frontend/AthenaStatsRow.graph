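# Athena frontend stats row widget, derived from FortProfileStatsRow, that displays one stat's name and value.
# PreConstruct writes the StatName text into the inherited Name block, and the SetStatValueAsText function writes the passed text into the inherited Value block.
# The layout stacks a thin dark-blue divider above a blue padded bar, with an uppercase, dimmed name filling the left and a larger value right-aligned.
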WidgetBlueprint /Game/Athena/UI/Frontend/AthenaStatsRow
graphs [
 EventGraph {nodes [
   Event PreConstruct()
   > "SetText (Text)"(self=Get Name() -> (Name:Common Text Object Reference), InText=Get StatName() -> (StatName:Text))
  ]}
 SetStatValueAsText {nodes [
   n1 = SetStatValueAsText!() -> (then, StatValue:Text) {is=title K2Node_FunctionEntry}
   > "SetText (Text)"(self=Get Value() -> (Value:Common Text Object Reference), InText=@n1)
  ]}
]
variables []
widgets {
 info {inheritedWidgetCount=2, parentClass=/Script/FortniteUI.FortProfileStatsRow, rootWidgetClass=&VerticalBox, widgetCount=7}
 widgets [
  VerticalBox_0 {widgetClassPath=&VerticalBox, widget=VerticalBox_0}
  CommonBorder_3 {
   widgetClassPath=&CommonBorder
   parent=VerticalBox_0
   widget=CommonBorder_3
   properties {style=/FortUILibrary/Styles/Border/Border_Solid_DkBlue.Border_Solid_DkBlue_C, brushColor=<rgba 1 1 1 0.7>, padding=<ltrb 0 0 0 0>}
   slot {class=VerticalBoxSlot}
  }
  Spacer_0 {
   widgetClassPath=&Spacer
   parent=CommonBorder_3
   widget=Spacer_0
   properties {size=<1 4>}
   slot {class=BorderSlot, properties {padding=<ltrb 0 0 0 0>}}
  }
  CommonBorder_1 {
   widgetClassPath=&CommonBorder
   parent=VerticalBox_0
   widget=CommonBorder_1
   properties {
    verticalAlignment=VAlign_Center
    background {
     bIsDynamicallyLoaded=false
     drawAs=Image
     imageType=NoImage
     mirroring=NoMirror
     resourceName=None
     resourceObject=null
     tiling=NoTile
     imageSize=<32 32>
     margin=<ltrb 0 0 0 0>
     outlineSettings {
      bUseBrushTransparency=false
      roundingType=HalfHeightRadius
      width=0
      color {colorUseRule=UseColor_Specified, specifiedColor=<rgba 0 0 0 0>}
      cornerRadii=<0 0 0 0>
     }
     tintColor {colorUseRule=UseColor_Specified, specifiedColor=<rgba 1 1 1 1>}
     uVRegion {bIsValid=false, max=<0 0>, min=<0 0>}
    }
    brushColor=<rgba 0.0009 0.0497 0.3185 1>
    padding=<ltrb 32 8 32 8>
   }
   slot {class=VerticalBoxSlot}
  }
  HorizontalBox_3 {
   widgetClassPath=&HorizontalBox
   parent=CommonBorder_1
   widget=HorizontalBox_3
   properties {flowDirectionPreference=Culture}
   slot {class=BorderSlot, properties {verticalAlignment=VAlign_Center, padding=<ltrb 32 8 32 8>}}
  }
  Name {
   widgetClassPath=&CommonTextBlock
   bInherited=true
   bIsVariable=true
   parent=HorizontalBox_3
   widget=Name
   properties {
    text=Kills
    textTransformPolicy=ToUpper
    colorAndOpacity {colorUseRule=UseColor_Specified, specifiedColor=<rgba 1 1 1 0.7>}
    font {
     bForceMonospaced=false
     bMaterialIsStencil=false
     fontMaterial=null
     fontObject=&BurbankSmall
     letterSpacing=0
     monospacedWidth=1
     size=16
     skewAmount=0
     typefaceFontName=Bold
     outlineSettings {
      bApplyOutlineToDropShadows=false
      bMiteredCorners=false
      bSeparateFillAlpha=false
      outlineMaterial=null
      outlineSize=0
      outlineColor=<rgba 0 0 0 1>
     }
    }
    shadowOffset=<0 0>
   }
   slot {class=HorizontalBoxSlot, properties {verticalAlignment=VAlign_Center, size {sizeRule=Fill, value=1}}}
  }
  Value {
   widgetClassPath=&CommonTextBlock
   bInherited=true
   bIsVariable=true
   parent=HorizontalBox_3
   widget=Value
   properties {
    text=N/A
    font {
     bForceMonospaced=false
     bMaterialIsStencil=false
     fontMaterial=null
     fontObject=&BurbankSmall
     letterSpacing=0
     monospacedWidth=1
     size=33
     skewAmount=0
     typefaceFontName=Black
     outlineSettings {
      bApplyOutlineToDropShadows=false
      bMiteredCorners=false
      bSeparateFillAlpha=false
      outlineMaterial=null
      outlineSize=0
      outlineColor=<rgba 0 0 0 1>
     }
    }
    shadowOffset=<0 0>
   }
   slot {class=HorizontalBoxSlot, properties {horizontalAlignment=HAlign_Right, verticalAlignment=VAlign_Center, padding=<ltrb 32 0 0 0>}}
  }
 ]
}
