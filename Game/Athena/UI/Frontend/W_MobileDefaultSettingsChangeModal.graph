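# Frontend modal dialog for a change to mobile default settings, built on the native MobileDefaultSettingsChangeModal class.
# Its only widget is a UIKit internal dialog, which receives a newly constructed MobileDefaultSettingsChangeModalVM as its data on Construct.
# The dialog's button-click dispatcher is bound on Construct and unbound on Destruct.
# Button index 0 calls ApplyDefaults and index 1 calls CloseModal.
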WidgetBlueprint /Game/Athena/UI/Frontend/W_MobileDefaultSettingsChangeModal
graphs [EventGraph {nodes [
   n1 = DialogButtonClicked!() -> (OutputDelegate, then, ButtonIndex:int) {is=title K2Node_CustomEvent}
   > n2 = Switch on Int!(execute:Exec, Selection:int=@n1.ButtonIndex) -> (Default:Exec, 0:Exec, 1:Exec)
   ApplyDefaults(execute=@n2.0, self:Mobile Default Settings Change Modal Object Reference)
   CloseModal(execute=@n2.1, self:Mobile Default Settings Change Modal Object Reference)
   Event Construct()
   n3 = Create Event()
   n4 = Get Dialog_Internal() -> (Dialog_Internal:WBP UIKit Dialog Internal Object Reference)
   > Bind Event to Dialog Button Clicked Dispatcher(self:WBP UIKit Dialog Internal Object Reference=@n4, Delegate=@n3) {is=title K2Node_AddDelegate}
   > n5 = Construct Object from Class(Class=&/MobileDefaultSettingsChangeModalVM.MobileDefaultSettingsChangeModalVM_C) -> (ReturnValue:Mobile Default Settings Change Modal VM Object Reference)
   > SetDialogData!(execute, self:WBP UIKit Dialog Internal Object Reference=Get Dialog_Internal(), Data:UI Kit Dialog Viewmodel Object Reference=@n5) -> (then)
   Event Destruct()
   > Unbind Event from Dialog Button Clicked Dispatcher(self:WBP UIKit Dialog Internal Object Reference=@n4, Delegate=@n3) {is=title K2Node_RemoveDelegate}
  ]}]
variables []
widgets {
 info {
  inheritedWidgetCount=0
  parentClass=/Script/FortniteUI.MobileDefaultSettingsChangeModal
  rootWidgetClass=/FortUILibrary/Systems/Dialog/WBP_UIKit_Dialog_Internal.WBP_UIKit_Dialog_Internal_C
  widgetCount=1
 }
 widgets [Dialog_Internal {
   widgetClassPath=/FortUILibrary/Systems/Dialog/WBP_UIKit_Dialog_Internal
   bIsVariable=true
   widget=Dialog_Internal
   properties {
    designTimePreview_DialogVM=&MobileDefaultSettingsChangeModalVM_C
    buttons {
     each Button Horizontal Fill=false
     each Button Min Width=157
     each Button Size=Medium
     focuse Button=0
     space Between Buttons=24
     cTA Buttons Types [Loud,Regular]
     group Placement {
      horizontal Alignment=HAlign_Fill
      vertical Alignment=VAlign_Bottom
      grid {column=1, column Span=0, layer=0, row=2, row Span=0, nudge=<0 0>}
      padding=<ltrb 80 64 80 48>
     }
    }
    dialogSIze {
     height Override {enable=true, value=500}
     max Aspect Ratio {enable=false, value=0}
     max Desired Height {enable=true, value=800}
     max Desired Width {enable=false, value=0}
     min Aspect Ratio {enable=false, value=0}
     min Desired Height {enable=false, value=0}
     min Desired Width {enable=false, value=0}
     width Override {enable=true, value=1200}
    }
    titleProperties {
     auto Wrap Text=true
     justification=Left
     line Height Percentage=1
     mobile Font Size Multiplier=1.3
     transform Policy=ToUpper
     wrap Text At=800
     color {colorUseRule=UseColor_Specified, specifiedColor=<rgba 1 1 1 1>}
     font {
      bForceMonospaced=false
      bMaterialIsStencil=false
      fontMaterial=null
      fontObject=&HeadingNow
      letterSpacing=0
      monospacedWidth=1
      size=25.5
      skewAmount=0
      typefaceFontName=86 Bold Italic
      outlineSettings {
       bApplyOutlineToDropShadows=false
       bMiteredCorners=false
       bSeparateFillAlpha=false
       outlineMaterial=null
       outlineSize=0
       outlineColor=<rgba 0 0 0 1>
      }
     }
     placement {
      horizontal Alignment=HAlign_Fill
      vertical Alignment=VAlign_Top
      grid {column=1, column Span=0, layer=0, row=0, row Span=0, nudge=<0 0>}
      padding=<ltrb 80 80 80 32>
     }
    }
   }
  }]
}
