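# AffiliateEntry is a frontend widget Blueprint based on the native FortAffiliateEntry class, providing the visual layout for a single affiliate entry.
# A horizontal box holds a padded, fixed-width size box containing the inherited OptionDisplayName text block, styled from the base text style with a NotoSans font override.
# The inherited InputActionWidget is not placed in the tree, and the event graph has only empty event nodes, so the Blueprint adds no logic.
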WidgetBlueprint /Game/Athena/UI/Frontend/AffiliateEntry
graphs [EventGraph {nodes [
   Event PreConstruct()
   Event Construct()
   Event Tick()
  ]}]
variables []
widgets {
 info {inheritedWidgetCount=2, parentClass=/Script/FortniteUI.FortAffiliateEntry, rootWidgetClass=&HorizontalBox, widgetCount=4}
 widgets [
  HorizontalBox_0 {widgetClassPath=&HorizontalBox, widget=HorizontalBox_0}
  SizeBox_0 {
   widgetClassPath=&SizeBox
   parent=HorizontalBox_0
   widget=SizeBox_0
   properties {widthOverride=417.5767}
   slot {class=HorizontalBoxSlot, properties {padding=<ltrb 10 5 10 5>}}
  }
  OptionDisplayName {
   widgetClassPath=&CommonTextBlock
   bInherited=true
   parent=SizeBox_0
   widget=OptionDisplayName
   properties {
    style=&TextStyle-Base_C
    text=Text Block
    font {
     bForceMonospaced=false
     bMaterialIsStencil=false
     fontMaterial=null
     fontObject=&NotoSans
     letterSpacing=0
     monospacedWidth=1
     size=24
     skewAmount=0
     typefaceFontName=NotoSans-Regular
     outlineSettings {
      bApplyOutlineToDropShadows=false
      bMiteredCorners=false
      bSeparateFillAlpha=false
      outlineMaterial=null
      outlineSize=0
      outlineColor=<rgba 0 0 0 1>
     }
    }
    shadowOffset=<0 0>
   }
   slot {class=SizeBoxSlot}
  }
  InputActionWidget {widgetClassPath=&CommonActionWidget, bInherited=true, widget=null}
 ]
}
