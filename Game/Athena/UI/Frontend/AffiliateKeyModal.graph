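# Support-A-Creator modal widget, built on FortAffilateModal, where players enter and save a creator code.
# Accept validates the code against a hotfixed runtime-options list or through MCP, with a 15-second timeout, then saves valid codes through MCP.
# Failures show entry or connection errors, with shakes for invalid or over-length input.
# Success plays a checkmark animation whose finish event pops the modal and broadcasts OnSuccess and OnPopupClosed.
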WidgetBlueprint /Game/Athena/UI/Frontend/AffiliateKeyModal
animations [
 Shake {
  binding_count=1
  display_label=""
  length_seconds=0.6
  object=&self:Shake
  sequence {bindings [AffiliateKey {
     id=#1
     possessed_class=WBP_UIKit_InputField_C
     tracks [{
       type=MovieScene2DTransformTrack
       sections [{
         type=MovieScene2DTransformSection
         end_frame=12
         start_frame=0
         channels [
          Translation.X {type=MovieSceneScriptingFloatChannel, keys [[0,0],[2,-5],[4,5],[6,-4],[8,4],[10,-3],[12,3]]}
          Translation.Y {type=MovieSceneScriptingFloatChannel, const [0,1]}
          Angle {type=MovieSceneScriptingFloatChannel, const [0,1]}
          Scale.X {type=MovieSceneScriptingFloatChannel, const [1,1]}
          Scale.Y {type=MovieSceneScriptingFloatChannel, const [1,1]}
          Shear.X {type=MovieSceneScriptingFloatChannel, const [0,1]}
          Shear.Y {type=MovieSceneScriptingFloatChannel, const [0,1]}
         ]
        }]
      }]
    }]}
 }
 SuccessAnim {
  binding_count=1
  length_seconds=0.75
  object=&self:SuccessAnim
  sequence {bindings [Check {
     id=#2
     possessed_class=Image
     tracks [
      {
       type=MovieScene2DTransformTrack
       sections [{
         type=MovieScene2DTransformSection
         end_frame=15
         start_frame=-1
         channels [
          Translation.X {type=MovieSceneScriptingFloatChannel}
          Translation.Y {type=MovieSceneScriptingFloatChannel, keys [[1,64],[3.3227,0]]}
          Angle {type=MovieSceneScriptingFloatChannel, keys [[1.0003,360],[4.484,0]]}
          Scale.X {type=MovieSceneScriptingFloatChannel, keys [[1,4],[4.484,0.75],[6.8067,1.25],[10.2907,1]]}
          Scale.Y {type=MovieSceneScriptingFloatChannel, keys [[1,4],[4.484,0.75],[6.8003,1.5],[6.8067,1.25],[10.2907,1]]}
          Shear.X {type=MovieSceneScriptingFloatChannel}
          Shear.Y {type=MovieSceneScriptingFloatChannel}
         ]
        }]
      }
      {
       type=MovieSceneColorTrack
       display_name=Color and Opacity
       sections [{
         type=MovieSceneColorSection
         end_frame=15
         start_frame=0
         channels [
          Color.R {type=MovieSceneScriptingFloatChannel}
          Color.G {type=MovieSceneScriptingFloatChannel}
          Color.B {type=MovieSceneScriptingFloatChannel}
          Color.A {type=MovieSceneScriptingFloatChannel, keys [[1,0],[2.0003,0],[4.0003,1]]}
         ]
        }]
      }
     ]
    }]}
 }
]
graphs [
 EventGraph {nodes [
   "OnTextChanged (Editable Text) (AffiliateKey)"!() -> (OutputDelegate, then, Text:"Text (by ref)") {is=title K2Node_ComponentBoundEvent}
   > n1 = LimitTextLength!(
    execute
    EditableTextWidget:Widget Object Reference=Get AffiliateKey() -> (AffiliateKey:Widget Object Reference)
    Length:int=Get MaxCharactersInTextEntry!(self:Fort Runtime Options Object Reference=GetRuntimeOptions()) -> (MaxCharactersInTextEntry:int=0)
   ) -> (then, WasTrimmed:bool)
   n2 = Get bShakeAnimationIsPlaying() -> (bShakeAnimationIsPlaying:bool)
   > Branch(Condition=AND Boolean(A=@n1, B=NOT Boolean(A=@n2)))
   > Set bShakeAnimationIsPlaying(bShakeAnimationIsPlaying:bool=true) -> (Output_Get:bool)
   > PlayAnimationByName!(
    execute
    self:Fort Simple Widget Animations Panel Object Reference=Get SimpleWidgetAnimationsPanel() -> (SimpleWidgetAnimationsPanel:Fort Simple Widget Animations Panel Object Reference)
    Name:"Name (by ref)"=Get ShakeAnimation() -> (ShakeAnimation:Name=None)
    Speed:float=1
    StartAtTime:float=0
   ) -> (then, ReturnValue:bool)
   > Delay(Duration=0.5)
   > Set bShakeAnimationIsPlaying(bShakeAnimationIsPlaying=false)
   n3 = "On Button Clicked (Button_CloseTouch)"!() -> (OutputDelegate, then, Button:Common Button Legacy Object Reference) {is=title K2Node_ComponentBoundEvent}
   PopContentWidget!(execute=[@n3.then, @n4.then]) -> (then)
   > Call On Popup Closed!(execute:Exec, self:Affiliate Key Modal Object Reference) -> (then:Exec) {is=title K2Node_CallDelegate}
   n4 = "On Clicked (CloseButton)"!() -> (OutputDelegate, then, Button:Common Button Base Object Reference) {is=title K2Node_ComponentBoundEvent}
   "On Clicked (LaunchWebsite)"!() -> (OutputDelegate, then, Button:Common Button Base Object Reference) {is=title K2Node_ComponentBoundEvent}
   n5 = Get ListOfCreatorsURL!(self:Fort Runtime Options Object Reference=GetRuntimeOptions()) -> (ListOfCreatorsURL:String)
   > n6 = CanLaunchURL!(execute, URL:String=@n5) -> (then, ReturnValue:bool)
   > Branch(Condition=@n6)
   > LaunchURL!(execute, URL:String=@n5) -> (then)
   n7 = "On Clicked (AcceptButton)"!() -> (OutputDelegate, then, Button:Common Button Base Object Reference) {is=title K2Node_ComponentBoundEvent}
   > SetIsEnabled(self=@n7.Button, bInIsEnabled=false)
   > SetVisibility(self=Get McpWaitingSpinner() -> (McpWaitingSpinner:Image Object Reference), InVisibility=SelfHitTestInvisible)
   > SetVisibility(self=Get ConnectionError() -> (ConnectionError:Common Text Object Reference), InVisibility=Collapsed)
   > n8 = Branch(Condition=Get bUseHotfixedAffiliateNamesArray!(self:Fort Runtime Options Object Reference=GetRuntimeOptions()) -> (bUseHotfixedAffiliateNamesArray:bool))
   n9 = "To String (Text)"(InText="GetText (Editable Text)"!(self:Editable Text Object Reference=Get EditableText() -> (EditableText:UIKit Editable Text Object Reference)) -> (ReturnValue:Text))
   > OnCheckComplete!(execute, self, bSuccess:bool=true, bWasValidAffiliateName:bool=false, AffilateNameChecked:String=@n9) -> (then)
   SetIsEnabled(execute=@n8.else, self=Get AcceptButton() -> (AcceptButton:Common Button Base Object Reference), bInIsEnabled=false)
   > Set ConnectionTimeOut(ConnectionTimeOut:bool=true) -> (Output_Get:bool)
   > CheckAffiliateName!(
    execute
    self:Fort Mcp Context Object Reference=Get FortMcpContext!() -> (ReturnValue:Fort Mcp Context Object Reference) {is=title K2Node_GetBlueprintContextEx}
    AffiliateName:String=@n9
    OnCheckAffiliateNameComplete:Delegate=Create Event()
   ) -> (then)
   > Delay(Duration=15)
   > Branch(Condition=Get ConnectionTimeOut() -> (ConnectionTimeOut:bool))
   > SetVisibility(self=Get ConnectionError(), InVisibility=SelfHitTestInvisible)
   > SetIsEnabled(self=Get AcceptButton(), bInIsEnabled=true)
   > SetVisibility(self=Get McpWaitingSpinner(), InVisibility=Hidden)
   n10 = OnCheckComplete!() -> (OutputDelegate, then, bSuccess:bool=~, bWasValidAffiliateName:bool=~, AffilateNameChecked:String) {is=title K2Node_CustomEvent}
   > Set ConnectionTimeOut(ConnectionTimeOut=false)
   > SetVisibility(self=Get McpWaitingSpinner(), InVisibility=Hidden)
   > n11 = Branch(Condition=@n10.bSuccess)
   > n12 = Branch(Condition=Get bUseHotfixedAffiliateNamesArray(self=GetRuntimeOptions()))
   n13 = GetValidAffiliateNames!(self:Fort Runtime Options Object Reference=GetRuntimeOptions()) -> (ReturnValue:Array of Strings)
   n14 = "To String (Text)"(InText="GetText (Editable Text)"(self=Get EditableText()))
   > n15 = Set Found(Found:bool="integer >= integer"!(A:int=Find Item!(TargetArray=@n13, ItemToFind:"String (by ref)"=@n14) -> (ReturnValue), B:int=0) -> (ReturnValue:bool) {is=title K2Node_CallFunction}) -> (Output_Get:bool)
   n16 = Get Found() -> (Found:bool)
   SetVisibility(
    execute=[@n15.then, @n21.then]
    self=[Get EntryError() -> (EntryError:Common Text Object Reference), Get ErrorPart2() -> (ErrorPart2:Common Text Object Reference)]
    InVisibility=Select(Option 0=SelfHitTestInvisible, Option 1=Collapsed, Index=@n16)
   )
   > n17 = Branch(Condition=@n16)
   > SetVisibility(self=Get McpWaitingSpinner(), InVisibility=SelfHitTestInvisible)
   > SetIsEnabled(self=Get AcceptButton(), bInIsEnabled=false)
   > n18 = Cast To FortPlayerController(Object=GetOwningPlayer())
   n19 = Get EditableText()
   n20 = OnSet!() -> (OutputDelegate, then, bSuccess:bool=~) {is=title K2Node_CustomEvent}
   SetAffiliateName!(
    execute=@n18.then
    self:Fort Mcp Context Object Reference=Get FortMcpContext()
    FortPC:Fort Player Controller Object Reference=@n18
    Name:String="To String (Text)"(InText="GetText (Editable Text)"(self=@n19))
    SourceName:String=Get SourceName() -> (SourceName:String)
    OnAffilateNameSetAcknowledged:Delegate=@n20
   ) -> (then)
   PlayAnimation(execute=@n17.else, InAnimation=Get Shake() -> (Shake:Widget Animation Object Reference), PlaybackSpeed=2)
   > SetIsEnabled(self=Get AcceptButton(), bInIsEnabled=true)
   n21 = Set Found(execute=@n12.else, Found=@n10.bWasValidAffiliateName)
   SetVisibility(execute=@n11.else, self=Get ConnectionError(), InVisibility=SelfHitTestInvisible)
   > SetIsEnabled(self=Get AcceptButton(), bInIsEnabled=true)
   Event OnInitialized()
   > n22 = Cast To WBP_UIKit_InputField_Base!(execute, Object=Get AffiliateKey()) -> (then, CastFailed, AsWBP UIKit Input Field Base:WBP UIKit Input Field Base Object Reference) {is=title K2Node_DynamicCast}
   > n23 = Set EditableText(EditableText:UIKit Editable Text Object Reference=Get EditableText!(self:WBP UIKit Block Editable Text Object Reference=Get Block_EditableText!(self:WBP UIKit Input Field Base Object Reference=@n22) -> (Block_EditableText:WBP UIKit Block Editable Text Object Reference)) -> (EditableText)) -> (Output_Get:UIKit Editable Text Object Reference)
   > "SetIsReadOnly (Editable Text)"!(execute, self:Editable Text Object Reference=@n23, InbIsReadyOnly:bool=Get bIsAffiliateKeyReadOnly() -> (bIsAffiliateKeyReadOnly:bool)) -> (then)
   Event On Activated()
   > SetOpacity!(execute, self:Image Object Reference=Get Check() -> (Check:Image Object Reference), InOpacity:float=0) -> (then)
   > SetIsEnabled(self=Get AcceptButton(), bInIsEnabled=true)
   > SetVisibility(self=Get McpWaitingSpinner(), InVisibility=Hidden)
   n24 = Len(S=Get ListOfCreatorsURL(self=GetRuntimeOptions()))
   > SetVisibility(
    self=Get LaunchWebsite() -> (LaunchWebsite:WBP UIKit Button Regular Object Reference)
    InVisibility=Select(Option 0=Collapsed, Option 1=SelfHitTestInvisible, Index="integer > integer"(A=@n24, B=0) {is=title K2Node_CallFunction})
   )
   n25 = Get ErrorPart2()
   n26 = Get ListOfCreatorsURL(self=GetRuntimeOptions())
   > "SetText (Text)"(
    self=@n25
    InText=Format Text!(Format="GetText (Text)"!(self:Text Object Reference=@n25) -> (ReturnValue:Text), 0:Text="To Text (String)"(InString=@n26)) -> (Result)
   )
   > n27 = Branch(Condition=Get bSkipAutoPopulate() -> (bSkipAutoPopulate:bool))
   n28 = Get FortMcpContext()
   "SetText (Editable Text)"!(
    execute=@n27.else
    self:Editable Text Object Reference=Get EditableText!() -> (EditableText)
    InText:Text="To Text (String)"(InString=GetAffiliateName!(self:Fort Mcp Context Object Reference=@n28) -> (ReturnValue:String))
   ) -> (then)
   "Animation Finished (SuccessAnim)"!() -> (OutputDelegate:Delegate, then:Exec) {is=title K2Node_WidgetAnimationEvent}
   > PopContentWidget()
   > Call On Success!(execute:Exec, self:Affiliate Key Modal Object Reference) -> (then:Exec) {is=title K2Node_CallDelegate}
   > Call On Popup Closed()
   SetVisibility(execute=@n30.then, self=Get ConnectionError(), InVisibility=SelfHitTestInvisible)
   PlayAnimation(execute=@n29.then, InAnimation=Get SuccessAnim() -> (SuccessAnim:Widget Animation Object Reference))
   SetVisibility(execute=@n20.then, self=Get McpWaitingSpinner(), InVisibility=Collapsed)
   > n29 = Branch(Condition=@n20.bSuccess)
   n30 = SetIsEnabled(execute=@n29.else, self=Get AcceptButton(), bInIsEnabled=true)
  ]}
 IsAffiliateKeyReadOnly {nodes [
   IsAffiliateKeyReadOnly() {is=title K2Node_FunctionEntry}
   > Return Node!(execute, bIsAffiliateKeyReadOnly:bool=Get bIsAffiliateKeyReadOnly())
  ]}
 OnPopupClosed {nodes [OnPopupClosed() {is=title K2Node_FunctionEntry}]}
 OnSuccess {nodes [OnSuccess() {is=title K2Node_FunctionEntry}]}
 SetupAffiliateField {nodes [
   n1 = SetupAffiliateField!() -> (then, AffilateName:String) {is=title K2Node_FunctionEntry}
   > "SetText (Editable Text)"(self=Get EditableText(), InText="To Text (String)"(InString=@n1))
  ]}
]
variables [
 bIsAffiliateKeyReadOnly {type=boolean}
 bShakeAnimationIsPlaying {type=boolean}
 bSkipAutoPopulate {type=boolean}
 ConnectionTimeOut {type=boolean}
 EditableText {type=UIKitEditableText, default=null}
 Found {type=boolean}
 OnPopupClosed {}
 OnSuccess {}
 ShakeAnimation {type=string, default=Shake_Simple}
]
widgets {
 info {inheritedWidgetCount=3, parentClass=/Script/FortniteUI.FortAffilateModal, rootWidgetClass=&BackgroundBlur, widgetCount=40}
 widgets [
  BackgroundBlur_0 {widgetClassPath=&BackgroundBlur, widget=BackgroundBlur_0, properties {blurStrength=25}}
  CommonBorder_0 {
   widgetClassPath=&CommonBorder
   parent=BackgroundBlur_0
   widget=CommonBorder_0
   properties {padding=<ltrb 0 0 0 0>}
   slot {class=BackgroundBlurSlot, properties {padding=<ltrb 0 0 0 0>}}
  }
  Overlay_0 {widgetClassPath=&Overlay, parent=CommonBorder_0, widget=Overlay_0, slot {class=BorderSlot, properties {padding=<ltrb 0 0 0 0>}}}
  Lightbox {
   widgetClassPath=/Game/UI/Foundation/Lightbox/Lightbox
   parent=Overlay_0
   widget=Lightbox
   properties {showWindowControlButtons=false}
   slot {class=OverlaySlot, properties {horizontalAlignment=HAlign_Fill, verticalAlignment=VAlign_Fill}}
  }
  VerticalBox_2 {
   widgetClassPath=&VerticalBox
   parent=Overlay_0
   widget=VerticalBox_2
   slot {class=OverlaySlot, properties {horizontalAlignment=HAlign_Center, verticalAlignment=VAlign_Fill}}
  }
  Overlay_85 {
   widgetClassPath=&Overlay
   parent=VerticalBox_2
   widget=Overlay_85
   slot {class=VerticalBoxSlot, properties {size {sizeRule=Fill, value=1}}}
  }
  SizeBox_0 {
   widgetClassPath=&SizeBox
   parent=VerticalBox_2
   widget=SizeBox_0
   properties {bOverride_MinDesiredWidth=true, heightOverride=800, minDesiredWidth=880, widthOverride=880}
   slot {class=VerticalBoxSlot}
  }
  CommonBorder_1 {
   widgetClassPath=&CommonBorder
   parent=SizeBox_0
   widget=CommonBorder_1
   properties {
    background {
     bIsDynamicallyLoaded=false
     drawAs=Image
     imageType=NoImage
     mirroring=NoMirror
     resourceName=None
     resourceObject=null
     tiling=NoTile
     imageSize=<32 32>
     margin=<ltrb 0 0 0 0>
     outlineSettings {
      bUseBrushTransparency=false
      roundingType=HalfHeightRadius
      width=0
      color {colorUseRule=UseColor_Specified, specifiedColor=<rgba 0 0 0 0>}
      cornerRadii=<0 0 0 0>
     }
     tintColor {colorUseRule=UseColor_Specified, specifiedColor=<rgba 0.0024 0.016 0.117 1>}
     uVRegion {bIsValid=false, max=<0 0>, min=<0 0>}
    }
    padding=<ltrb 0 0 0 0>
   }
   slot {class=SizeBoxSlot}
  }
  Overlay_2 {widgetClassPath=&Overlay, parent=CommonBorder_1, widget=Overlay_2, slot {class=BorderSlot, properties {padding=<ltrb 0 0 0 0>}}}
  WBP_CaptureForPostBufferUpdate {
   widgetClassPath=&WBP_CaptureForPostBufferUpdate
   bIsVariable=true
   parent=Overlay_2
   slot {class=OverlaySlot, properties {horizontalAlignment=HAlign_Fill, verticalAlignment=VAlign_Fill}}
  }
  ModalContainer {
   widgetClassPath=&VerticalBox
   parent=Overlay_2
   widget=ModalContainer
   slot {class=OverlaySlot, properties {horizontalAlignment=HAlign_Fill, verticalAlignment=VAlign_Fill}}
  }
  Header {
   widgetClassPath=&CommonBorder
   parent=ModalContainer
   widget=Header
   properties {
    style=/FortUILibrary/Styles/Backgrounds/_BorderStyle_ModalHeader._BorderStyle_ModalHeader_C
    background {
     bIsDynamicallyLoaded=false
     drawAs=Box
     imageType=NoImage
     mirroring=NoMirror
     resourceName=None
     resourceObject=/FortUILibrary/Styles/Border/Materials/M-Border-ModalHeader-procedural
     tiling=NoTile
     imageSize=<64 64>
     margin=<ltrb 0.25 0.25 0.25 0.25>
     outlineSettings {
      bUseBrushTransparency=false
      roundingType=HalfHeightRadius
      width=0
      color {colorUseRule=UseColor_Specified, specifiedColor=<rgba 0 0 0 0>}
      cornerRadii=<0 0 0 0>
     }
     tintColor {colorUseRule=UseColor_Specified, specifiedColor=<rgba 1 1 1 1>}
     uVRegion {bIsValid=false, max=<0 0>, min=<0 0>}
    }
    padding=<ltrb 40 0 4 0>
   }
   slot {class=VerticalBoxSlot}
  }
  HorizontalBox_1 {
   widgetClassPath=&HorizontalBox
   parent=Header
   widget=HorizontalBox_1
   slot {class=BorderSlot, properties {padding=<ltrb 40 0 4 0>}}
  }
  Spacer_89 {
   widgetClassPath=&Spacer
   parent=HorizontalBox_1
   widget=Spacer_89
   properties {size=<0 85>}
   slot {class=HorizontalBoxSlot, properties {horizontalAlignment=HAlign_Left, verticalAlignment=VAlign_Top}}
  }
  CommonTextBlock_Title {
   widgetClassPath=&CommonTextBlock
   bIsVariable=true
   parent=HorizontalBox_1
   widget=CommonTextBlock_Title
   properties {
    clipping=ClipToBounds
    scrollStyle=&TextScrollStyle-Base_C
    style=/FortUILibrary/Styles/Typography/_TextStyle_SubH2._TextStyle_SubH2_C
    text=SUPPORT A CREATOR
    textTransformPolicy=ToUpper
    font {
     bForceMonospaced=false
     bMaterialIsStencil=false
     fontMaterial=null
     fontObject=&Burbank
     letterSpacing=0
     monospacedWidth=1
     size=33
     skewAmount=0
     typefaceFontName=BurbankRegular-Black
     outlineSettings {
      bApplyOutlineToDropShadows=false
      bMiteredCorners=false
      bSeparateFillAlpha=false
      outlineMaterial=null
      outlineSize=0
      outlineColor=<rgba 0 0 0 1>
     }
    }
    shadowOffset=<0 0>
    strikeBrush {
     bIsDynamicallyLoaded=false
     drawAs=NoDrawType
     imageType=NoImage
     mirroring=NoMirror
     resourceName=None
     resourceObject=null
     tiling=NoTile
     imageSize=<32 32>
     margin=<ltrb 0 0 0 0>
     outlineSettings {
      bUseBrushTransparency=false
      roundingType=HalfHeightRadius
      width=0
      color {colorUseRule=UseColor_Specified, specifiedColor=<rgba 0 0 0 0>}
      cornerRadii=<0 0 0 0>
     }
     tintColor {colorUseRule=UseColor_Specified, specifiedColor=<rgba 1 1 1 1>}
     uVRegion {bIsValid=false, max=<0 0>, min=<0 0>}
    }
   }
   slot {class=HorizontalBoxSlot, properties {verticalAlignment=VAlign_Center, padding=<ltrb 8 0 20 0>, size {sizeRule=Fill, value=1}}}
  }
  VisibilityBorder_Touch {
   widgetClassPath=&CommonHardwareVisibilityBorder
   parent=HorizontalBox_1
   widget=VisibilityBorder_Touch
   properties {visibilityQuery {
     autoDescription=" ANY( Input.Touch )"
     tokenStreamVersion=0
     userDescription=""
     queryTokenStream [0,1,1,1,0]
     tagDictionary [{tagName=Input.Touch}]
    }}
   slot {class=HorizontalBoxSlot, properties {horizontalAlignment=HAlign_Right, verticalAlignment=VAlign_Top, padding=<ltrb 0 10 10 10>}}
  }
  Button_CloseTouch {
   widgetClassPath=/FortUILibrary/Foundation/Buttons/CloseButton
   bIsVariable=true
   parent=VisibilityBorder_Touch
   widget=Button_CloseTouch
   properties {
    onButtonBaseClicked=(&self:WidgetTree.Button_CloseTouch.HandleOnButtonClicked)
    onButtonBaseDoubleClicked=(&self:WidgetTree.Button_CloseTouch.HandleOnButtonDoubleClicked)
    onButtonBaseHovered=(&self:WidgetTree.Button_CloseTouch.HandleOnButtonHovered)
    onButtonBaseUnhovered=(&self:WidgetTree.Button_CloseTouch.HandleOnButtonUnhovered)
    onSelectedChangedBase=(&self:WidgetTree.Button_CloseTouch.HandleOnSelectedChanged)
    pressedSlateSoundOverride {resourceObject=/Game/Sounds/Fort_UI_Sounds/FrontEnd/Whooshes/Toasts/UI_Whoosh_Generic_Toast_Close_Cue}
   }
   slot {class=BorderSlot, properties {padding=<ltrb 0 0 0 0>}}
  }
  Content {
   widgetClassPath=&VerticalBox
   parent=ModalContainer
   widget=Content
   properties {navigation=&self:WidgetTree.Content.WidgetNavigation_1}
   slot {class=VerticalBoxSlot, properties {padding=<ltrb 48 16 48 16>, size {sizeRule=Fill, value=1}}}
  }
  Description {
   widgetClassPath=&CommonTextBlock
   bIsVariable=true
   parent=Content
   widget=Description
   properties {
    autoWrapText=true
    lineHeightPercentage=1.29
    style=/FortUILibrary/Styles/Typography/_TextStyle_Body1._TextStyle_Body1_C
    text="Declare your support for a Creator! Your in-game purchases will help support this Creator."
    wrapTextAt=900
    font {
     bForceMonospaced=false
     bMaterialIsStencil=false
     fontMaterial=null
     fontObject=&BurbankSmall
     letterSpacing=0
     monospacedWidth=1
     size=18
     skewAmount=0
     typefaceFontName=Bold
     outlineSettings {
      bApplyOutlineToDropShadows=false
      bMiteredCorners=false
      bSeparateFillAlpha=false
      outlineMaterial=null
      outlineSize=0
      outlineColor=<rgba 0 0 0 1>
     }
    }
    shadowOffset=<0 0>
    strikeBrush {
     bIsDynamicallyLoaded=false
     drawAs=NoDrawType
     imageType=NoImage
     mirroring=NoMirror
     resourceName=None
     resourceObject=null
     tiling=NoTile
     imageSize=<32 32>
     margin=<ltrb 0 0 0 0>
     outlineSettings {
      bUseBrushTransparency=false
      roundingType=HalfHeightRadius
      width=0
      color {colorUseRule=UseColor_Specified, specifiedColor=<rgba 0 0 0 0>}
      cornerRadii=<0 0 0 0>
     }
     tintColor {colorUseRule=UseColor_Specified, specifiedColor=<rgba 1 1 1 1>}
     uVRegion {bIsValid=false, max=<0 0>, min=<0 0>}
    }
   }
   slot {class=VerticalBoxSlot, properties {padding=<ltrb 0 16 0 16>}}
  }
  SimpleWidgetAnimationsPanel {
   widgetClassPath=/Script/FortniteUI.FortSimpleWidgetAnimationsPanel
   bIsVariable=true
   parent=Content
   widget=SimpleWidgetAnimationsPanel
   properties {animations [Shake_Simple {
      bIsEnabled=true
      onAnimationsFinished=()
      targets [{
        alpha=null
        bResetOnFinish=true
        bShouldAnimateAlpha=false
        bShouldAnimateScale=false
        bShouldAnimateTranslation=true
        scale=null
        targetName=Overlay_1
        translation=/Game/UI/Utilities/Curve_ShakeAnim
        materialScalarAnimations []
       }]
     }]}
   slot {class=VerticalBoxSlot, properties {padding=<ltrb 0 5 0 5>}}
  }
  Overlay_1 {widgetClassPath=&Overlay, parent=SimpleWidgetAnimationsPanel, widget=Overlay_1, slot {class=FortSimpleWidgetAnimationsSlot}}
  AffiliateKey {
   widgetClassPath=/FortUILibrary/UIKit/InputField/WBP_UIKit_InputField
   bInherited=true
   parent=Overlay_1
   widget=AffiliateKey
   properties {hint Text="", holdData=&CommonUI_HoldData_Base_C, useIcon=false}
   slot {class=OverlaySlot, properties {horizontalAlignment=HAlign_Fill, verticalAlignment=VAlign_Fill}}
  }
  VerticalBox_0 {
   widgetClassPath=&VerticalBox
   parent=Content
   widget=VerticalBox_0
   slot {class=VerticalBoxSlot, properties {padding=<ltrb 0 0 0 10>}}
  }
  EntryError {
   widgetClassPath=&CommonTextBlock
   bIsVariable=true
   parent=VerticalBox_0
   widget=EntryError
   properties {
    style=/FortUILibrary/Styles/Typography/Foundation/TextStyle-Base-XS-B-S-Red.TextStyle-Base-XS-B-S-Red_C
    text="Creator not found. "
    visibility=Hidden
    colorAndOpacity {colorUseRule=UseColor_Specified, specifiedColor=<rgba 1 0.0561 0.0648 1>}
    font {
     bForceMonospaced=false
     bMaterialIsStencil=false
     fontMaterial=null
     fontObject=&NotoSans
     letterSpacing=0
     monospacedWidth=1
     size=16
     skewAmount=0
     typefaceFontName=NotoSans-Bold
     outlineSettings {
      bApplyOutlineToDropShadows=true
      bMiteredCorners=false
      bSeparateFillAlpha=true
      outlineMaterial=&M-FontShadow
      outlineSize=2
      outlineColor=<rgba 0.085 0.0146 0.0177 0.5>
     }
    }
    shadowOffset=<0 0>
   }
   slot {class=VerticalBoxSlot}
  }
  ConnectionError {
   widgetClassPath=&CommonTextBlock
   bIsVariable=true
   parent=VerticalBox_0
   widget=ConnectionError
   properties {
    style=&TextStyle-Base-XS-B-S-Red_C
    text="Connection error, please try again later."
    visibility=Collapsed
    colorAndOpacity {colorUseRule=UseColor_Specified, specifiedColor=<rgba 1 0.0561 0.0648 1>}
    font {
     bForceMonospaced=false
     bMaterialIsStencil=false
     fontMaterial=null
     fontObject=&NotoSans
     letterSpacing=0
     monospacedWidth=1
     size=16
     skewAmount=0
     typefaceFontName=NotoSans-Bold
     outlineSettings {
      bApplyOutlineToDropShadows=true
      bMiteredCorners=false
      bSeparateFillAlpha=true
      outlineMaterial=&M-FontShadow
      outlineSize=2
      outlineColor=<rgba 0.085 0.0146 0.0177 0.5>
     }
    }
    shadowOffset=<0 0>
   }
   slot {class=VerticalBoxSlot}
  }
  HorizontalBox_3 {widgetClassPath=&HorizontalBox, parent=VerticalBox_0, widget=HorizontalBox_3, slot {class=VerticalBoxSlot}}
  BorderCheck {
   widgetClassPath=&Border
   parent=HorizontalBox_3
   widget=BorderCheck
   properties {background {
     bIsDynamicallyLoaded=false
     drawAs=NoDrawType
     imageType=NoImage
     mirroring=NoMirror
     resourceName=None
     resourceObject=null
     tiling=NoTile
     imageSize=<32 32>
     margin=<ltrb 0 0 0 0>
     outlineSettings {
      bUseBrushTransparency=false
      roundingType=HalfHeightRadius
      width=0
      color {colorUseRule=UseColor_Specified, specifiedColor=<rgba 0 0 0 0>}
      cornerRadii=<0 0 0 0>
     }
     tintColor {colorUseRule=UseColor_Specified, specifiedColor=<rgba 1 1 1 1>}
     uVRegion {bIsValid=false, max=<0 0>, min=<0 0>}
    }}
   slot {class=HorizontalBoxSlot, properties {horizontalAlignment=HAlign_Right, verticalAlignment=VAlign_Center}}
  }
  CommonTextBlock_0 {
   widgetClassPath=&CommonTextBlock
   bIsVariable=true
   parent=VerticalBox_0
   widget=CommonTextBlock_0
   properties {
    style=/Game/UI/Foundation/Items/Parts/TextStyles/ItemCardPowerRatingTextStyle_XS.ItemCardPowerRatingTextStyle_XS_C
    text=Try one of these suggestions:
    visibility=Collapsed
    font {
     bForceMonospaced=false
     bMaterialIsStencil=false
     fontMaterial=null
     fontObject=&NotoSans
     letterSpacing=0
     monospacedWidth=1
     size=15
     skewAmount=0
     typefaceFontName=NotoSans-Bold
     outlineSettings {
      bApplyOutlineToDropShadows=false
      bMiteredCorners=false
      bSeparateFillAlpha=false
      outlineMaterial=null
      outlineSize=0
      outlineColor=<rgba 0 0 0 1>
     }
    }
    shadowOffset=<0 0>
   }
   slot {class=VerticalBoxSlot, properties {padding=<ltrb 2 0 0 0>}}
  }
  AffiliateSuggestionList {
   widgetClassPath=&CommonListView
   bInherited=true
   bIsVariable=true
   parent=VerticalBox_0
   widget=AffiliateSuggestionList
   properties {
    entryWidgetClass=&/AffiliateEntry.AffiliateEntry_C
    navigation=&self:WidgetTree.AffiliateSuggestionList.WidgetNavigation_0
    visibility=Collapsed
   }
   slot {class=VerticalBoxSlot, properties {horizontalAlignment=HAlign_Left, padding=<ltrb 20 20 20 20>}}
  }
  ErrorPart2 {
   widgetClassPath=&CommonTextBlock
   bIsVariable=true
   parent=VerticalBox_0
   widget=ErrorPart2
   properties {
    autoWrapText=true
    style=&ItemCardPowerRatingTextStyle_XS_C
    text="Please ensure the spelling is correct.\r\n\r\nNote: Not all content creators and streamers are currently in the Creator program. More coming soon!"
    visibility=Collapsed
    font {
     bForceMonospaced=false
     bMaterialIsStencil=false
     fontMaterial=null
     fontObject=&NotoSans
     letterSpacing=0
     monospacedWidth=1
     size=15
     skewAmount=0
     typefaceFontName=NotoSans-Bold
     outlineSettings {
      bApplyOutlineToDropShadows=false
      bMiteredCorners=false
      bSeparateFillAlpha=false
      outlineMaterial=null
      outlineSize=0
      outlineColor=<rgba 0 0 0 1>
     }
    }
    shadowOffset=<0 0>
   }
   slot {class=VerticalBoxSlot, properties {padding=<ltrb 2 0 0 6>}}
  }
  Overlay_3 {
   widgetClassPath=&Overlay
   parent=ModalContainer
   widget=Overlay_3
   slot {class=VerticalBoxSlot, properties {verticalAlignment=VAlign_Bottom}}
  }
  SizeBox_1 {
   widgetClassPath=&SizeBox
   parent=Overlay_3
   widget=SizeBox_1
   properties {minDesiredWidth=646.5314}
   slot {class=OverlaySlot, properties {horizontalAlignment=HAlign_Right, padding=<ltrb 12 12 30 20>}}
  }
  HorizontalBox_0 {widgetClassPath=&HorizontalBox, parent=SizeBox_1, widget=HorizontalBox_0, slot {class=SizeBoxSlot}}
  Check {
   widgetClassPath=&Image
   bIsVariable=true
   parent=HorizontalBox_0
   widget=Check
   properties {
    brush {
     bIsDynamicallyLoaded=false
     drawAs=Image
     imageType=NoImage
     mirroring=NoMirror
     resourceName=None
     resourceObject=/Game/UI/Foundation/Textures/Icons/Checks/T-Icon-Check-64
     tiling=NoTile
     imageSize=<48 48>
     margin=<ltrb 0 0 0 0>
     outlineSettings {
      bUseBrushTransparency=false
      roundingType=HalfHeightRadius
      width=0
      color {colorUseRule=UseColor_Specified, specifiedColor=<rgba 0 0 0 0>}
      cornerRadii=<0 0 0 0>
     }
     tintColor {colorUseRule=UseColor_Specified, specifiedColor=<rgba 1 1 1 1>}
     uVRegion {bIsValid=false, max=<0 0>, min=<0 0>}
    }
    colorAndOpacity=<rgba 0.07 0.7 0.175 0>
   }
   slot {class=HorizontalBoxSlot, properties {horizontalAlignment=HAlign_Center, verticalAlignment=VAlign_Center, padding=<ltrb 10 0 10 0>}}
  }
  McpWaitingSpinner {
   widgetClassPath=&Image
   bIsVariable=true
   parent=HorizontalBox_0
   widget=McpWaitingSpinner
   properties {
    visibility=Collapsed
    brush {
     bIsDynamicallyLoaded=false
     drawAs=Image
     imageType=NoImage
     mirroring=NoMirror
     resourceName=None
     resourceObject=/FortUILibrary/UIKit/Throbber/MI_UIKit_Monochrome_Throbber
     tiling=NoTile
     imageSize=<40 40>
     margin=<ltrb 0 0 0 0>
     outlineSettings {
      bUseBrushTransparency=false
      roundingType=HalfHeightRadius
      width=0
      color {colorUseRule=UseColor_Specified, specifiedColor=<rgba 0 0 0 0>}
      cornerRadii=<0 0 0 0>
     }
     tintColor {colorUseRule=UseColor_Specified, specifiedColor=<rgba 1 1 1 1>}
     uVRegion {bIsValid=false, max=<0 0>, min=<0 0>}
    }
   }
   slot {class=HorizontalBoxSlot, properties {verticalAlignment=VAlign_Center, padding=<ltrb 0 0 16 0>}}
  }
  LaunchWebsite {
   widgetClassPath=&WBP_UIKit_Button_Regular
   bIsVariable=true
   parent=HorizontalBox_0
   widget=LaunchWebsite
   properties {
    bIsFocusable=false
    minHeight=54
    pressedSound=/Game/Sounds/Fort_UI_Sounds/FrontEnd/Whooshes/Toasts/UI_Whoosh_Tiny_Appear_Cue
    text=View approved creators
    visibility=Collapsed
    triggeringInputAction {dataTable=&InputActionDataTable, rowName=CreatorsWebLink}
   }
   slot {class=HorizontalBoxSlot, properties {padding=<ltrb 0 0 32 0>}}
  }
  AcceptButton {
   widgetClassPath=/FortUILibrary/UIKit/Buttons/WBP_UIKit_Button_Loud
   bInherited=true
   bIsVariable=true
   parent=HorizontalBox_0
   widget=AcceptButton
   properties {
    bIsFocusable=false
    pressedSound=/Game/Sounds/Fort_UI_Sounds/FrontEnd/Whooshes/Toasts/UI_Whoosh_Tiny_Close_Cue
    text=Accept
    triggeringInputAction {dataTable=&InputActionDataTable, rowName=Accept}
   }
   slot {class=HorizontalBoxSlot, properties {padding=<ltrb 0 0 32 0>}}
  }
  VisibilityBorder_NonTouch {
   widgetClassPath=&CommonHardwareVisibilityBorder
   parent=HorizontalBox_0
   widget=VisibilityBorder_NonTouch
   properties {visibilityQuery {
     autoDescription=" NONE( Input.Touch )"
     tokenStreamVersion=0
     userDescription=""
     queryTokenStream [0,1,3,1,0]
     tagDictionary [{tagName=Input.Touch}]
    }}
   slot {class=HorizontalBoxSlot}
  }
  CloseButton {
   widgetClassPath=&WBP_UIKit_Button_Regular
   bIsVariable=true
   parent=VisibilityBorder_NonTouch
   widget=CloseButton
   properties {
    bIsFocusable=false
    pressedSound=&UI_Whoosh_Generic_Toast_Close_Cue
    text=Close
    pressedSlateSoundOverride {resourceObject=/Game/Sounds/Fort_UI_Sounds/FrontEnd/New_FrontEnd/DefaultSounds/PressedButtons/FN_UI_Default_Button_Pressed_Cue}
    triggeringInputAction {dataTable=&InputActionDataTable, rowName=Close}
   }
   slot {class=BorderSlot, properties {padding=<ltrb 0 0 0 0>}}
  }
  Warning {
   widgetClassPath=&CommonTextBlock
   bIsVariable=true
   parent=VerticalBox_2
   widget=Warning
   properties {
    autoWrapText=true
    justification=Center
    renderOpacity=0.7
    style=/FortUILibrary/Styles/Typography/_TextStyle_Legal._TextStyle_Legal_C
    text="By using a Creator Code you are supporting the creator with your purchase. You will never receive anything from using a Creator Code. If someone is claiming you will receive a benefit, it's a scam - please file a ticket with Player Support."
    visibility=HitTestInvisible
    wrapTextAt=1050
    font {
     bForceMonospaced=false
     bMaterialIsStencil=false
     fontMaterial=null
     fontObject=&NotoSans
     letterSpacing=0
     monospacedWidth=1
     size=15
     skewAmount=0
     typefaceFontName=NotoSans-Regular
     outlineSettings {
      bApplyOutlineToDropShadows=false
      bMiteredCorners=false
      bSeparateFillAlpha=false
      outlineMaterial=null
      outlineSize=0
      outlineColor=<rgba 0 0 0 1>
     }
    }
    shadowOffset=<0 0>
    strikeBrush {
     bIsDynamicallyLoaded=false
     drawAs=NoDrawType
     imageType=NoImage
     mirroring=NoMirror
     resourceName=None
     resourceObject=null
     tiling=NoTile
     imageSize=<32 32>
     margin=<ltrb 0 0 0 0>
     outlineSettings {
      bUseBrushTransparency=false
      roundingType=HalfHeightRadius
      width=0
      color {colorUseRule=UseColor_Specified, specifiedColor=<rgba 0 0 0 0>}
      cornerRadii=<0 0 0 0>
     }
     tintColor {colorUseRule=UseColor_Specified, specifiedColor=<rgba 1 1 1 1>}
     uVRegion {bIsValid=false, max=<0 0>, min=<0 0>}
    }
   }
   slot {class=VerticalBoxSlot, properties {verticalAlignment=VAlign_Center, size {sizeRule=Fill, value=1}}}
  }
 ]
}
